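# Editor utility widget for tuning first-person mesh camera-offset material parameters on the player during PIE.
# GetActor reads the pawn's mesh materials, dedupes dynamic instances by parent, and collects scalars and parent static switches sourced from MF_FPMeshCameraOffset.
# Details-view edits push scalars to the dynamic instances and switches to parent constant instances.
# PromoteCurrentValuesToMIs writes scalars into parents and saves them, temporarily enabling saving during PIE.
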EditorUtilityWidgetBlueprint /Game/NeverCook/ArtTools/Tech_Art/FPTools/EUW_GrabFPSMeshComponentMaterial
graphs [
 EventGraph {nodes [
   "On Clicked (GetCurrentPlayer)"!() -> (OutputDelegate, then, widget:User Widget Object Reference, event:Pointer Event Structure) {is=title K2Node_ComponentBoundEvent}
   > n1 = GetPIEWorlds!(execute, bIncludeDedicatedServer:bool=false) -> (then, ReturnValue:Array of World Object References)
   > Clear(TargetArray:Array of EUW FPSMesh Component Params Structures=Get Materials() -> (Materials:Array of EUW FPSMesh Component Params Structures))
   > Clear(TargetArray:Array of EUW Grab FPSMIStatic Switch Data Structures=Get MIsToUpdate() -> (MIsToUpdate:Array of EUW Grab FPSMIStatic Switch Data Structures))
   n2 = GetPlayerPawn!(
    WorldContextObject:Object Reference="Get (a copy)"(Array:Array of World Object References=@n1) -> (Output:World Object Reference)
    PlayerIndex:int=0
   ) -> (ReturnValue:Pawn Object Reference)
   > n3 = Cast To FortPlayerPawn(Object=@n2)
   > n4 = Cast To SkeletalMeshComponentBudgeted!(execute, Object=GetSkeletalMeshForPartType(self=@n3)) -> (then, CastFailed, AsSkeletal Mesh Component Budgeted:Skeletal Mesh Component Budgeted Object Reference) {is=title K2Node_DynamicCast}
   n5 = For Each Loop(
    Exec=[@n4.then, @n23.Is Valid]
    Array:Array of Material Interface Object References=GetMaterials!(self:Mesh Component Object Reference=@n4) -> (ReturnValue:Array of Material Interface Object References)
   ) -> (Array Element:Material Interface Object Reference)
   n6 = For Each Loop!(Exec=@n5.Completed, Array:Array of EUW FPSMesh Component Params Structures=Get Materials()) -> (
    LoopBody
    Array Element_Material_4_#1:Material Interface Object Reference
    Array Element_ScalarParams_6_#2:Array of EUW Grab FPSVector Scalar Param Structures
    Array Index
    Completed
   )
   n7 = Get MIsToUpdate()
   n8 = For Each Loop!(Exec=@n6.Completed, Array:Array of EUW Grab FPSMIStatic Switch Data Structures=@n7) -> (LoopBody, Array Element:EUW Grab FPSMIStatic Switch Data Structure, Array Index, Completed)
   > GetStaticSwitchs!(
    execute
    self
    Material:"EUW Grab FPSMIStatic Switch Data Structure (by ref)"="Get (a ref)"(Array:Array of EUW Grab FPSMIStatic Switch Data Structures=@n7, Dimension 1=@n8.Array Index) -> (Output:"EUW Grab FPSMIStatic Switch Data Structure (by ref)")
   ) -> (then)
   n9 = Cast To MaterialInstanceDynamic!(execute=@n6.LoopBody, Object=@n6) -> (then, CastFailed, AsMaterial Instance Dynamic:Material Instance Dynamic Object Reference) {is=title K2Node_DynamicCast}
   > Add Unique!(
    execute
    TargetArray:Array of EUW Grab FPSMIStatic Switch Data Structures=Get MIsToUpdate()
    NewItem_MaterialInstance_2_#3:Material Interface Object Reference=Get Parent!(self:Material Instance Object Reference=@n9) -> (Parent:Material Interface Object Reference)
    NewItem_StaticSwitches_6_#4:Array of EUW Grab FPSStatic Switch Params Structures
   ) -> (then, ReturnValue)
   n10 = VerifyClearToAdd!(execute=@n5.LoopBody, self, MaterialInterface:Material Interface Object Reference=@n5) -> (then, SafeToAdd?:bool)
   > Branch(Condition=@n10)
   > n11 = GetScalarParams!(execute, self, Material:Material Interface Object Reference=@n5) -> (then, ScalarParams:Array of EUW Grab FPSVector Scalar Param Structures)
   > Add!(
    execute
    TargetArray:Array of EUW FPSMesh Component Params Structures=Get Materials()
    NewItem_Material_4_#1:Material Interface Object Reference=@n5
    NewItem_ScalarParams_6_#2:Array of EUW Grab FPSVector Scalar Param Structures=@n11
   ) -> (then, ReturnValue) {is=title K2Node_CallArrayFunction}
   "On Property Changed (StaticSwitchParamValues)"!() -> (OutputDelegate, then, PropertyName:Name=None) {is=title K2Node_ComponentBoundEvent}
   > n12 = For Each Loop!(Exec, Array=Get MIsToUpdate()) -> (
    LoopBody
    Array Element_MaterialInstance_2_#3:Material Interface Object Reference
    Array Element_StaticSwitches_6_#4:Array of EUW Grab FPSStatic Switch Params Structures
    Array Index
    Completed
   )
   > n13 = Cast To MaterialInstanceConstant(Object=@n12)
   > n14 = For Each Loop!(Exec, Array:Array of EUW Grab FPSStatic Switch Params Structures=@n12.Array Element_StaticSwitches_6_#4) -> (
    LoopBody
    Array Element_StaticSwitch_2_#5:Name
    Array Element_Value_5_#6:bool=~
    Array Index
    Completed
   )
   > SetMaterialInstanceStaticSwitchParameterValue!(
    execute
    Instance:Material Instance Constant Object Reference=@n13
    ParameterName:Name=@n14
    Value:bool=@n14.Array Element_Value_5_#6
    Association:EMaterialParameterAssociation Enum=GlobalParameter
    bUpdateMaterialInstance:bool=true
   ) -> (then, ReturnValue:bool)
   "On Property Changed (ScalerParamValues)"!() -> (OutputDelegate, then, PropertyName:Name=None) {is=title K2Node_ComponentBoundEvent}
   > n15 = For Each Loop!(Exec, Array:Array of EUW FPSMesh Component Params Structures=Get Materials()) -> (
    LoopBody
    Array Element_Material_4_#1:Material Interface Object Reference
    Array Element_ScalarParams_6_#2:Array of EUW Grab FPSVector Scalar Param Structures
    Array Index
    Completed
   )
   > n16 = Cast To MaterialInstanceDynamic(Object=@n15)
   > n17 = For Each Loop!(Exec, Array:Array of EUW Grab FPSVector Scalar Param Structures=@n15.Array Element_ScalarParams_6_#2) -> (
    LoopBody
    Array Element_ParamName_5_#7:Name
    Array Element_Value_4_#8:double=~
    Array Index
    Completed
   )
   > SetScalarParameterValue(self=@n16, ParameterName=@n17, Value=@n17.Array Element_Value_4_#8)
   "On Clicked (SaveCurrentValuesToMI)"!() -> (OutputDelegate, then, widget:User Widget Object Reference, event:Pointer Event Structure) {is=title K2Node_ComponentBoundEvent}
   > n18 = For Each Loop!(Exec, Array:Array of EUW FPSMesh Component Params Structures=Get Materials()) -> (
    LoopBody
    Array Element_Material_4_#1:Material Interface Object Reference
    Array Element_ScalarParams_6_#2:Array of EUW Grab FPSVector Scalar Param Structures
    Array Index
    Completed
   )
   > n19 = Cast To MaterialInstanceDynamic(Object=@n18)
   > n20 = Cast To MaterialInstanceConstant(Object=Get Parent(self=@n19))
   > n21 = For Each Loop!(Exec, Array:Array of EUW Grab FPSVector Scalar Param Structures=@n18.Array Element_ScalarParams_6_#2) -> (
    LoopBody
    Array Element_ParamName_5_#7:Name
    Array Element_Value_4_#8:double=~
    Array Index
    Completed
   )
   ExecuteConsoleCommandFromEditor!(execute=@n21.Completed, ConsoleCommand:String=Editor.AllowSavingAssetsDuringPIE 1) -> (then)
   > SaveLoadedAsset(AssetToSave=@n20, bOnlyIfIsDirty=false)
   > ExecuteConsoleCommandFromEditor(ConsoleCommand=Editor.AllowSavingAssetsDuringPIE 0)
   n22 = Branch(
    execute=@n21.LoopBody
    Condition="Equal (Float)"(
     A=@n21.Array Element_Value_4_#8
     B=GetMaterialInstanceScalarParameterValue!(
      Instance:Material Instance Constant Object Reference=@n20
      ParameterName:Name=@n21
      Association:EMaterialParameterAssociation Enum=GlobalParameter
     ) -> (ReturnValue:float)
    )
   )
   SetMaterialInstanceScalarParameterValue!(
    execute=@n22.else
    Instance:Material Instance Constant Object Reference=@n20
    ParameterName:Name=@n21
    Value:float=@n21.Array Element_Value_4_#8
    Association:EMaterialParameterAssociation Enum=GlobalParameter
   ) -> (then, ReturnValue:bool)
   > MarkPropertyDirty!(execute, Object:Object Reference=@n20, PropertyName:Name=@n21) -> (then)
   PrintString(execute=@n20.CastFailed, InString=FailedCastToconstant)
   PrintString(execute=@n19.CastFailed, InString=FailedCastToDynamic)
   Event PreConstruct()
   > SetObject(
    self=[
     Get StaticSwitchParamValues() -> (StaticSwitchParamValues:Details View Object Reference)
     Get ScalerParamValues() -> (ScalerParamValues:Details View Object Reference)
    ]
    NewObject=Self-Reference()
   )
   n23 = Is Valid(InputObject=FindComponentByTag!(
     self:Actor Object Reference=@n2
     ComponentClass:Actor Component Class Reference=/Script/AnimationBudgetAllocator.SkeletalMeshComponentBudgeted
     Tag:Name=Gameplay.Camera.FirstPersonPlayerMesh
    ) -> (ReturnValue:Skeletal Mesh Component Budgeted Object Reference))
   PrintString(execute=@n23.Is Not Valid, InString="COULD NOT FIND Gameplay.Camera.FirstPersonPlayerMesh COMPONENT!")
   Event Construct()
   Event Tick()
  ]}
 GetScalarParams {nodes [
   n1 = GetScalarParams!() -> (then, Material:Material Interface Object Reference) {is=title K2Node_FunctionEntry}
   > n2 = GetScalarParameterNames!(execute, Material:Material Interface Object Reference=@n1) -> (then, ParameterNames:Array of Names)
   > n3 = For Each Loop!(Exec, Array:Array of Names=@n2) -> (LoopBody, Array Element:Name, Array Index, Completed)
   n4 = Get NewLocalVar() -> (NewLocalVar:Array of EUW Grab FPSVector Scalar Param Structures)
   Return Node!(execute=@n3.Completed, ScalarParams:Array of EUW Grab FPSVector Scalar Param Structures=@n4)
   > n5 = GetScalarParameterSource!(execute, Material:Material Interface Object Reference=@n1, ParameterName:Name=@n3) -> (then, ParameterSource_PathString:String, ReturnValue:bool)
   > Branch(Condition=AND Boolean(A=Contains(SearchIn=@n5, Substring=MF_FPMeshCameraOffset), B="Not Equal (Name)"(A=@n3, B=FP_WPOStrength) -> (ReturnValue=~)))
   > n6 = Cast To MaterialInstanceDynamic(Object=@n1)
   > n7 = Get Scalar Parameter Value!(execute, self:Material Instance Dynamic Object Reference=@n6, ParameterName:Name=@n3) -> (then, ReturnValue:float) {is=title K2Node_CallFunction}
   > Add!(
    execute
    TargetArray:Array of EUW Grab FPSVector Scalar Param Structures=@n4
    NewItem_ParamName_5_#7:Name=@n3
    NewItem_Value_4_#8:double=@n7
   ) -> (then, ReturnValue)
  ]}
 GetStaticSwitchs {nodes [
   n1 = GetStaticSwitchs!() -> (then, Material:"EUW Grab FPSMIStatic Switch Data Structure (by ref)") {is=title K2Node_FunctionEntry}
   n2 = Break EUW Grab FPSMIStatic Switch Data!(EUW_GrabFPSMIStaticSwitchData:"EUW Grab FPSMIStatic Switch Data Structure (by ref)"=@n1) -> (
    MaterialInstance_2_#3:Material Interface Object Reference
    StaticSwitches_6_#4:Array of EUW Grab FPSStatic Switch Params Structures
   ) {is=title K2Node_BreakStruct}
   > n3 = GetStaticSwitchParameterNames!(execute, Material:Material Interface Object Reference=@n2) -> (then, ParameterNames:Array of Names)
   > n4 = For Each Loop(Array=@n3)
   n5 = For Each Loop(Exec=@n4.Completed, Array=Get SwitchNames() -> (SwitchNames:Array of Names))
   Set members in EUW Grab FPSMIStatic Switch Data!(
    execute:Exec=@n5.Completed
    StructRef:"EUW Grab FPSMIStatic Switch Data Structure (by ref)"=@n1
    StaticSwitches_6_#4:Array of EUW Grab FPSStatic Switch Params Structures=Get StaticSwitchData() -> (StaticSwitchData:Array of EUW Grab FPSStatic Switch Params Structures)
   ) -> (then:Exec, StructOut:"EUW Grab FPSMIStatic Switch Data Structure (by ref)") {is=title K2Node_SetFieldsInStruct}
   Add!(
    execute=@n5.LoopBody
    TargetArray:Array of EUW Grab FPSStatic Switch Params Structures=Get StaticSwitchData()
    NewItem_StaticSwitch_2_#5:Name=@n5
    NewItem_Value_5_#6:bool="Get (a copy)"(Array:Array of Booleans=Get SwitchValues() -> (SwitchValues:Array of Booleans), Dimension 1=@n5.Array Index) -> (Output:bool)
   ) -> (then, ReturnValue)
   n6 = GetStaticSwitchParameterSource!(execute=@n4.LoopBody, Material:Material Interface Object Reference=@n2, ParameterName:Name=@n4) -> (then, ParameterSource_PathString:String, ReturnValue:bool)
   > Branch(Condition=Contains(SearchIn=@n6, Substring=MF_FPMeshCameraOffset))
   > n7 = Cast To MaterialInstanceConstant(Object=@n2)
   > Add!(execute, TargetArray:Array of Names=Get SwitchNames(), NewItem:"Name (by ref)"=@n4) -> (then, ReturnValue)
   > Add(
    TargetArray:Array of Booleans=Get SwitchValues()
    NewItem:"Boolean (by ref)"=GetMaterialInstanceStaticSwitchParameterValue!(
     Instance:Material Instance Constant Object Reference=@n7
     ParameterName:Name=@n4
     Association:EMaterialParameterAssociation Enum=GlobalParameter
    ) -> (ReturnValue:bool)
   )
  ]}
 VerifyClearToAdd {nodes [
   n1 = VerifyClearToAdd!() -> (then, MaterialInterface:Material Interface Object Reference) {is=title K2Node_FunctionEntry}
   > n2 = Cast To MaterialInstanceDynamic(Object=@n1)
   n3 = Get Parent(self=@n2)
   > n4 = Branch(Condition="Equal (Object)"(A=@n3, B=/Engine/EngineMaterials/WorldGridMaterial))
   Return Node!(execute=[@n6.then, @n4.then, @n2.CastFailed], SafeToAdd?:bool=false)
   n5 = For Each Loop!(Exec=@n4.else, Array:Array of EUW FPSMesh Component Params Structures=Get Materials()) -> (
    LoopBody
    Array Element_Material_4_#1:Material Interface Object Reference
    Array Element_ScalarParams_6_#2:Array of EUW Grab FPSVector Scalar Param Structures
    Array Index
    Completed
   )
   n6 = Branch(execute=@n5.Completed, Condition=Get ParentMIAlreadyPresent() -> (ParentMIAlreadyPresent:bool))
   Return Node(execute=@n6.else, SafeToAdd?=true)
   n7 = Cast To MaterialInstanceDynamic(execute=@n5.LoopBody, Object=@n5)
   > Branch(Condition="Equal (Object)"(A=@n3, B=Get Parent(self=@n7)))
   > Set ParentMIAlreadyPresent(ParentMIAlreadyPresent:bool=true) -> (Output_Get:bool)
   Return Node(SafeToAdd?=false)
  ]}
]
variables [
 Materials {type="array<EUW_FPSMeshComponentParams>", default []}
 MIsToUpdate {type="array<EUW_GrabFPSMIStaticSwitchData>", default []}
]
widgets {
 info {rootWidgetClass=&EditorUtilityScrollBox, widgetCount=8}
 widgets [
  EditorUtilityScrollBox_0 {widgetClassPath=&EditorUtilityScrollBox, widget=EditorUtilityScrollBox_0}
  VerticalBox_0 {widgetClassPath=&VerticalBox, parent=EditorUtilityScrollBox_0, widget=VerticalBox_0, slot {class=ScrollBoxSlot}}
  HorizontalBox_1 {widgetClassPath=&HorizontalBox, parent=VerticalBox_0, widget=HorizontalBox_1, slot {class=VerticalBoxSlot}}
  GetCurrentPlayer {
   widgetClassPath=&TA_Generic01_Button
   parent=HorizontalBox_1
   widget=GetCurrentPlayer
   properties {label=GetActor}
   slot {class=HorizontalBoxSlot, properties {size {sizeRule=Fill, value=1}}}
  }
  SaveCurrentValuesToMI {
   widgetClassPath=&TA_Generic01_Button
   parent=HorizontalBox_1
   widget=SaveCurrentValuesToMI
   properties {label=PromoteCurrentValuesToMIs}
   slot {class=HorizontalBoxSlot, properties {size {sizeRule=Fill, value=1}}}
  }
  HorizontalBox_0 {
   widgetClassPath=&HorizontalBox
   parent=VerticalBox_0
   widget=HorizontalBox_0
   slot {class=VerticalBoxSlot, properties {size {sizeRule=Fill, value=1}}}
  }
  ScalerParamValues {
   widgetClassPath=&DetailsView
   parent=HorizontalBox_0
   widget=ScalerParamValues
   properties {categoriesToShow [DetailsView]}
   slot {class=HorizontalBoxSlot, properties {size {sizeRule=Fill, value=1}}}
  }
  StaticSwitchParamValues {
   widgetClassPath=&DetailsView
   parent=HorizontalBox_0
   widget=StaticSwitchParamValues
   properties {categoriesToShow [StaticSwitches]}
   slot {class=HorizontalBoxSlot, properties {size {sizeRule=Fill, value=1}}}
  }
 ]
}
